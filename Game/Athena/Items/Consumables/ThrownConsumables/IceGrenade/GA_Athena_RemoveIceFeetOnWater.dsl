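; Gameplay ability that removes Icy Feet status effects from its owning Fortnite pawn when that pawn starts surface swimming.
; On activation it caches the pawn and binds to its movement-mode-changed delegate, reacting to events rather than polling.
; When the custom movement mode equals SurfaceSwimming, it removes effects granting the Trap, Surface or Grenade IcyFeet tags.
; On ability end it unbinds the delegate if the pawn is still valid.
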(event Ability|EventActivateAbility
  (bind (_event_data _fort_player_pawn) (Setup|SetupGenericAbility false false)
    (:Output
      (bind _output_get (Variables|Default|SetFort%20Player%20Pawn _fort_player_pawn))
      (Character|BindEventtoMovementModeChangedDelegate _output_get (delegate Custom|On%20Movement%20Mode%20Changed)))))

(event Custom|On%20Movement%20Mode%20Changed
  (params (Character :type Character%20Object%20Reference) (PrevMovementMode :type EMovementMode%20Enum :default "MOVE_None") (PreviousCustomMode :type Byte))
  (if (Utilities|Enum|Equal(Enum) (Utilities|Enum|BytetoEnumEFortCustomMovement (Variables|/Script/Engine.CharacterMovementComponent|GetCustomMovementMode (Variables|/Script/Engine.Character|GetCharacterMovement Character))) "SurfaceSwimming")
    (Ability|RemoveGameplayEffectFromOwnerWithGrantedTags "(GameplayTags=((TagName=\"Gameplay.Status.IcyFeet.Trap\"),(TagName=\"Gameplay.Status.IcyFeet.Surface\"),(TagName=\"Gameplay.Status.IcyFeet.Grenade\")))")))

(event Ability|EventOnEndAbility
  (params (bWasCancelled :type Boolean))
  (bind _fort_player_pawn (Variables|Default|GetFort%20Player%20Pawn))
  (Utilities|IsValid _fort_player_pawn
    (:Is%20Valid
      (Character|UnbindEventfromMovementModeChangedDelegate _fort_player_pawn (delegate Custom|On%20Movement%20Mode%20Changed)))))
